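# Unlit translucent Niagara particle material for the SnowFlake vehicle wheel cosmetic VFX, drawing a distorted, cloud-like sprite.
# A slowly panning cloud noise texture offsets the UVs of a FlameCloud texture by a fixed amount.
# The clamped red channel is multiplied by particle color and a dynamic Brightness parameter for color, and by particle alpha with depth fade for soft opacity.
# The Distortion AMount dynamic parameter is declared but not used.
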Material /VehicleCosmetics/Wheels/SnowFlake/VFX/M_Snowflake_VFX
graph {
 properties {
  bAutomaticallySetUsageInEditor=false
  bUsedWithNiagaraMeshParticles=true
  bUsedWithNiagaraRibbons=true
  bUsedWithNiagaraSprites=true
  blendMode=BLEND_Translucent
  previewMesh=/VehicleCosmetics/Wheels/Vented/Mesh/SK_Wheel_Vented
  shadingModel=MSM_Unlit
  usedShadingModels=Unlit
 }
 nodes [
  n1 = Panner(Coordinate=TextureCoordinate() {uTiling=0.25, vTiling=0.25}, Time=Multiply(A=Time(), B=Constant() {r=0.5})) {speedX=0.1, speedY=0.1}
  n2 = Multiply(
   A=ComponentMask(None=TextureSample(UVs=@n1) {texture=/VehicleCosmetics/Wheels/_SharedTextures/FX/Clouds_RGB_Soft_03_T}.RGB) {g=true, r=true}
   B=Constant() {r=0.1}
  )
  n3 = TextureCoordinate()
  n4 = Clamp(None=TextureSample(UVs=Add(A=@n2, B=@n3)) {texture=&/FlameCloud_01}.R)
  n5 = ParticleColor()
  DepthFade(Opacity=Multiply(A=@n4, B=@n5.A)) {fadeDistanceDefault=50}
  Multiply(A=Multiply(A=@n4, B=@n5.RGB), B=DynamicParameter() {paramNames [Distortion AMount, Brightness, Param3, Param4]}.Brightness)
 ]
}
